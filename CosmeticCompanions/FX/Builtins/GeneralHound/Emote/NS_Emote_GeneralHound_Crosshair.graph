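# Red crosshair reticle effect for the GeneralHound companion emote, built from two near-identical emitters whose sprites face opposite ways.
# An initial burst drives the reticle, which pops in and out with squash-and-stretch curves, and a short second burst fires as it collapses.
# Spawn group sets lifetime and visibility tag.
# Crosshair001 adds extra sprite layers and a pulsed light, scaled for front-end and time of day; both emitters' SkeletalMeshLocation modules are disabled.
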NiagaraSystem /CosmeticCompanions/FX/Builtins/GeneralHound/Emote/NS_Emote_GeneralHound_Crosshair
dependencies {
 usedDataInterfaces [&NiagaraDataInterfaceCurve, &NiagaraDataInterfaceSkeletalMesh]
 usedDynamicInputs []
 usedModules [
  &self:SystemSpawnScript.SystemScriptSource.SystemScriptGraph.NiagaraNodeAssignment_0.SetVariables_#1
  &self:SystemSpawnScript.SystemScriptSource.SystemScriptGraph.NiagaraNodeAssignment_1.SetVariables_#2
  &self:SystemSpawnScript.SystemScriptSource.SystemScriptGraph.NiagaraNodeAssignment_2.SetVariables_#3
  &self:SystemSpawnScript.SystemScriptSource.SystemScriptGraph.NiagaraNodeAssignment_3.SetVariables_#4
  &self:SystemSpawnScript.SystemScriptSource.SystemScriptGraph.NiagaraNodeAssignment_4.SetVariables_#5
  &self:SystemSpawnScript.SystemScriptSource.SystemScriptGraph.NiagaraNodeAssignment_5.SetVariables_#6
  &SystemState
 ]
}
emitters {
 Crosshair {
  emitterUpdateScript [
   &EmitterState()
   &SpawnBurst_Instantaneous(Spawn Count=5, Spawn Group=1, Spawn Time=0.5, Use Spawn Probability=false, UseLoopCountLimit=false) {enabled=false, module=SpawnBurst_Instantaneous002}
   &SpawnBurst_Instantaneous(Spawn Count=5, Spawn Group=1, Spawn Time=1.72, Use Spawn Probability=false, UseLoopCountLimit=false) {module=SpawnBurst_Instantaneous001}
   &SpawnBurst_Instantaneous(Spawn Time=0.01, Use Spawn Probability=false, UseLoopCountLimit=false)
   /SetVariables(
    Emitter.Layer 1 - Main reticle:LinearColor=<rgba 1 0 0 1>
    Emitter.Layer 2 - Reticle outline glow:LinearColor=<rgba 1 0.3127 0 1>
    Emitter.Layer 4 - outerglow:LinearColor=<rgba 0.3118 0.4223 1 1>
   ) {isSetParameters=true, module=SetVariables_#7}
  ]
  particleSpawnScript [
   /SetVariables(Particles.VisibilityTag=&SpawnGroup()) {isSetParameters=true, module=SetVariables_#8}
   &InitializeParticle(
    Lifetime=&MergeHelper_FloatSwitch(0=2.75, 1=0.3, 2=?, 3=?, Index=&SpawnGroup(), Num=::NewEnumerator0)
    Mass Max=2
    Mass Mode=::NewEnumerator3
    Mesh Scale=<2 1.4 2>
    Mesh Scale Mode=::NewEnumerator4
    Sprite Rotation Mode=::NewEnumerator1
    Sprite Size Mode=::NewEnumerator3
    Uniform Sprite Size=45
    UsePositionOffset=true
    Write Material Random=true
   )
   &ApplyInitialForces()
   &SkeletalMeshLocation(
    Bone Sampling Mode=::NewEnumerator4
    MaskSpawnGroup=false
    Sampled Orientation Coordinate Space=::Local
    Sampled Position Coordinate Space=::Local
    Sampled Position Offset=&MergeHelper_VectorSwitch(0=<0 0 0>, 1=<0 0 0>, 2=?, 3=?, Index=&SpawnGroup(), Num=::NewEnumerator0)
    Sampled Position Offset Coordinate Space:ENiagaraSkelMeshTransforms=::NewEnumerator2
    Sampled Velocity Coordinate Space:ENiagaraCoordinateSpace=::Local
    Skeletal Mesh=$System.Skeletal Mesh Reader
   ) {enabled=false}
   /SetVariables() {isSetParameters=true, module=SetVariables_#9}
  ]
  particleUpdateScript [
   &SolveForcesAndVelocity(Clamp Velocity=false, Limit Acceleration=false)
   &ParticleState()
   &ScaleColor(
    Scale Alpha=&FloatFromCurve(FloatCurve=DataInterface() {propertyValues {Curve {keys [
         {arriveTangent=0.002, interpMode=RCIM_Cubic, time=0, value=1.0113}
         {arriveTangent=-2.705, interpMode=RCIM_Cubic, leaveTangent=-2.705, tangentMode=RCTM_User, time=1.0011, value=0}
        ]}}})
    ScaleA=true
    ScaleRGB=false
   )
   /SetVariables(Particles.SpriteSize=&Multiply_Vector2D(
     A=$Particles.Initial.SpriteSize
     B=&MakeVector2D(
      X=&FloatFromCurve(FloatCurve=DataInterface() {propertyValues {Curve {keys [
           <at 0.015 0 RCIM_Cubic>
           {arriveTangent=33.3333, interpMode=RCIM_Cubic, leaveTangent=33.3333, time=0.03, value=1.3}
           {arriveTangent=-0.5263, interpMode=RCIM_Cubic, leaveTangent=-0.5263, time=0.045, value=1}
           {arriveTangent=0.5217, interpMode=RCIM_Cubic, leaveTangent=0.5217, time=0.6, value=1}
           {arriveTangent=-0.3208, interpMode=RCIM_Cubic, leaveTangent=-0.3208, tangentMode=RCTM_User, time=0.62, value=1.3}
           <at 0.65 0 RCIM_Cubic>
          ]}}})
      Y=&FloatFromCurve(FloatCurve=DataInterface() {propertyValues {Curve {keys [
           <at 0 2 RCIM_Cubic>
           {arriveTangent=-22.2222, interpMode=RCIM_Cubic, leaveTangent=-22.2222, time=0.025, value=0.8}
           {arriveTangent=0.3478, interpMode=RCIM_Cubic, leaveTangent=0.3478, time=0.045, value=1}
           {arriveTangent=-0.3478, interpMode=RCIM_Cubic, leaveTangent=-0.3478, time=0.6, value=1}
           {arriveTangent=0.2037, interpMode=RCIM_Cubic, leaveTangent=0.2037, tangentMode=RCTM_User, time=0.62, value=0.8}
           <at 0.65 2 RCIM_Cubic>
          ]}}})
     )
    )) {isSetParameters=true, module=SetVariables_#10}
   /SetVariables(
    Particles.Croshair main color:LinearColor=&MakeLinearColorFromVectorAndFloat(
     "Vector (RGB)"=&MakeVectorFromLinearColorRGB(Color=&FNTimeOfDayColor(Night Color=<rgba 1 0 0.0001 1>, Morning Color=<rgba 1 0 0.0001 1>, Day Color=<rgba 1 0 0.0001 1>, Evening Color=<rgba 1 0 0.0001 1>))
     "Float (Alpha)"=&FloatFromCurve(FloatCurve=DataInterface() {propertyValues {Curve {keys [{arriveTangent=0.002, interpMode=RCIM_Cubic, time=0, value=1.0113}, <at 1.0011 0 RCIM_Cubic>]}}})
    )
    Particles.Crosshair SpriteFacing:Vector3f=&Multiply_Vector(A=Linked() {linkedVariable {name=Particles.SkeletalMeshLocation.SampledPosition, type=NiagaraPosition}}, B=<1 0 0>)
    Particles.Crosshair glow SpriteSize:Vector2f=&Multiply_Vector2D(
     A=$Particles.Initial.SpriteSize
     B=&MakeVector2D(
      X=&FloatFromCurve(
       FloatCurve=DataInterface() {propertyValues {Curve {keys [
           <at 0.015 0 RCIM_Cubic>
           {arriveTangent=33.3333, interpMode=RCIM_Cubic, leaveTangent=33.3333, time=0.03, value=1.3}
           {arriveTangent=-0.5172, interpMode=RCIM_Cubic, leaveTangent=-0.5172, time=0.045, value=1}
           {arriveTangent=-0.3208, interpMode=RCIM_Cubic, leaveTangent=-0.3208, tangentMode=RCTM_User, time=0.61, value=1}
           <at 0.64 0 RCIM_Cubic>
          ]}}}
       Scale Curve=2.2
      )
      Y=&FloatFromCurve(
       FloatCurve=DataInterface() {propertyValues {Curve {keys [
           <at 0 2 RCIM_Cubic>
           {arriveTangent=-22.2222, interpMode=RCIM_Cubic, leaveTangent=-22.2222, time=0.025, value=0.8}
           {arriveTangent=0.3361, interpMode=RCIM_Cubic, leaveTangent=0.3361, time=0.045, value=1}
           {arriveTangent=0.2037, interpMode=RCIM_Cubic, leaveTangent=0.2037, tangentMode=RCTM_User, time=0.62, value=1}
           <at 0.635 2 RCIM_Cubic>
          ]}}}
       Scale Curve=2.2
      )
     )
    )
    Particles.Crosshair glow color:LinearColor=&MakeLinearColorFromVectorAndFloat(
     "Vector (RGB)"=&MakeVectorFromLinearColorRGB(Color=&FNTimeOfDayColor(Night Color=<rgba 1 0 0.0001 1>, Morning Color=<rgba 1 0 0.0001 1>, Day Color=<rgba 1 0 0.0001 1>, Evening Color=<rgba 1 0 0.0001 1>))
     "Float (Alpha)"=&FloatFromCurve(
      FloatCurve=DataInterface() {propertyValues {Curve {keys [{arriveTangent=0.002, interpMode=RCIM_Cubic, time=0, value=1.0113}, <at 0.9 0 RCIM_Cubic>]}}}
      Scale Curve=0.5
     )
    )
    Particles.Crosshair opposite SpriteFacing:Vector3f=<0 0 0>
    Particles.CrosshairGlow:LinearColor=&MakeLinearColorFromVectorAndFloat(
     "Vector (RGB)"=&MakeVectorFromLinearColorRGB(Color=&FNTimeOfDayColor(
       Night Color=<rgba 5 0.257 0.0079 1>
       Morning Color=<rgba 5 0.257 0.0079 1>
       Day Color=<rgba 5 0.257 0.0079 1>
       Evening Color=<rgba 5 0.257 0.0079 1>
      ))
     "Float (Alpha)"=&FloatFromCurve(
      FloatCurve=DataInterface() {propertyValues {Curve {keys [
          {arriveTangent=0.002, interpMode=RCIM_Cubic, time=0, value=1}
          <at 0.2549 -0.0093 RCIM_Cubic>
          {arriveTangent=0.0127, interpMode=RCIM_Cubic, leaveTangent=0.0127, time=0.5, value=1}
          <at 0.9918 0 RCIM_Cubic>
         ]}}}
      Scale Curve=0.5
     )
    )
    Particles.White details:LinearColor=&MakeLinearColorFromVectorAndFloat(
     "Vector (RGB)"=&MakeVectorFromLinearColorRGB(Color=&FNTimeOfDayColor(Night Color=<rgba 5 5 5 1>, Morning Color=<rgba 5 5 5 1>, Day Color=<rgba 5 5 5 1>, Evening Color=<rgba 5 5 5 1>))
     "Float (Alpha)"=&FloatFromCurve(
      FloatCurve=DataInterface() {propertyValues {Curve {keys [
          {arriveTangent=0.002, interpMode=RCIM_Cubic, time=0, value=1}
          <at 0.2549 -0.0093 RCIM_Cubic>
          {arriveTangent=0.0127, interpMode=RCIM_Cubic, leaveTangent=0.0127, time=0.5, value=1}
          <at 0.9918 0 RCIM_Cubic>
         ]}}}
      Scale Curve=0.5
     )
    )
    Particles.crosshair glow aligned SpriteSize:Vector2f=&Multiply_Vector2D(
     A=$Particles.Initial.SpriteSize
     B=&MakeVector2D(
      X=&FloatFromCurve(
       FloatCurve=DataInterface() {propertyValues {Curve {keys [
           <at 0.015 0 RCIM_Cubic>
           {arriveTangent=33.3333, interpMode=RCIM_Cubic, leaveTangent=33.3333, time=0.03, value=1.3}
           {arriveTangent=-0.5172, interpMode=RCIM_Cubic, leaveTangent=-0.5172, time=0.045, value=1}
           {arriveTangent=-0.3208, interpMode=RCIM_Cubic, leaveTangent=-0.3208, tangentMode=RCTM_User, time=0.61, value=1}
           <at 0.64 0 RCIM_Cubic>
          ]}}}
       Scale Curve=1.5
      )
      Y=&FloatFromCurve(
       FloatCurve=DataInterface() {propertyValues {Curve {keys [
           <at 0 2 RCIM_Cubic>
           {arriveTangent=-22.2222, interpMode=RCIM_Cubic, leaveTangent=-22.2222, time=0.025, value=0.8}
           {arriveTangent=0.3361, interpMode=RCIM_Cubic, leaveTangent=0.3361, time=0.045, value=1}
           {arriveTangent=0.2037, interpMode=RCIM_Cubic, leaveTangent=0.2037, tangentMode=RCTM_User, time=0.62, value=1}
           <at 0.635 2 RCIM_Cubic>
          ]}}}
       Scale Curve=1.5
      )
     )
    )
   ) {isSetParameters=true, module=SetVariables_#11}
   /SetVariables(Particles.Crosshair Main DynParam:Vector4f=&MakeVector4(X=0, Y=2, Z=0, W=0)) {isSetParameters=true, module=SetVariables_#12}
   /SetVariables(
    Particles.Flare Position:NiagaraPosition=&AddVectorToPosition(Position=$Particles.SkeletalMeshLocation.SampledPosition, Vector=<0 0 0>)
    Particles.Flare SpriteSize:Vector2f=&Multiply_Vector2D(
     A=$Particles.Initial.SpriteSize
     B=&MakeVector2D(
      X=&FloatFromCurve(
       FloatCurve=DataInterface() {propertyValues {Curve {defaultValue=1, keys [{tangentMode=RCTM_User, time=0, value=1}, {tangentMode=RCTM_User, time=1, value=0}]}}}
       Scale Curve=5
      )
      Y=&FloatFromCurve(FloatCurve=DataInterface() {propertyValues {Curve {keys [<at 0.0122 0.8471>,<at 1 0>]}}}, Scale Curve=5)
     )
    )
    Particles.Sparkle Linear Color:LinearColor=&MakeLinearColorFromVectorAndFloat(
     "Vector (RGB)"=&MakeVectorFromLinearColorRGB(Color=&FNTimeOfDayColor(Night Color=<rgba 1 0 0.0001 1>, Morning Color=<rgba 1 0 0.0001 1>, Day Color=<rgba 1 0 0.0001 1>, Evening Color=<rgba 1 0 0.0001 1>))
     "Float (Alpha)"=&FloatFromCurve(FloatCurve=DataInterface() {propertyValues {Curve {keys [{arriveTangent=0.002, interpMode=RCIM_Cubic, time=0, value=1.0113}, <at 0.9 0 RCIM_Cubic>]}}})
    )
    Particles.Sparkle SpriteRotation:NiagaraFloat=&Multiply_Float(A=$Particles.NormalizedAge, B=60)
    Particles.Sparkle SpriteSize:Vector2f=&Multiply_Vector2D(
     A=$Particles.Initial.SpriteSize
     B=&Multiply_Vector2D_ByFloat(
      Float=&FloatFromCurve(FloatCurve=DataInterface() {propertyValues {Curve {keys [<at 0 1.5>,<at 1 0>]}}})
      Vector2D=<1 1>
     )
    )
   ) {isSetParameters=true, module=SetVariables_#13}
   /SetVariables(
    Particles.DynParam 0:Vector4f=&MakeVector4(X=1, Y=0, Z=0.1, W=0.5)
    Particles.Dynparam 1:Vector4f=&MakeVector4(X=1, Y=1, Z=1, W=1)
   ) {isSetParameters=true, module=SetVariables_#14}
   &FN_Light_Attributes(
    Color=<rgba 1 0 0 1>
    Create Color Attribute=true
    Create Exponent Attribute=true
    Create Position Attribute=true
    Create Radius Attribute=true
    Enable Particle Light Enable=true
    Enable Visibility Tag=true
    Exponent=6
    Light Enabled=true
    Position:NiagaraPosition=$Particles.SkeletalMeshLocation.SampledPosition
    Radius=&Multiply_Float(
     A=30
     B=&FloatFromCurve(FloatCurve=DataInterface() {propertyValues {Curve {keys [
          <at 0.015 0 RCIM_Cubic>
          {arriveTangent=33.3333, interpMode=RCIM_Cubic, leaveTangent=33.3333, time=0.03, value=1.3}
          {arriveTangent=-0.5263, interpMode=RCIM_Cubic, leaveTangent=-0.5263, time=0.045, value=1}
          {arriveTangent=0.5217, interpMode=RCIM_Cubic, leaveTangent=0.5217, time=0.6, value=1}
          {arriveTangent=-0.3208, interpMode=RCIM_Cubic, leaveTangent=-0.3208, tangentMode=RCTM_User, time=0.62, value=1.3}
          <at 0.65 0 RCIM_Cubic>
         ]}}})
    )
    Scale Color=&FloatFromCurve(
     FloatCurve=DataInterface() {propertyValues {Curve {keys [
         {arriveTangent=5.626, interpMode=RCIM_Cubic, leaveTangent=5.626, tangentMode=RCTM_User, time=0, value=0.6}
         {arriveTangent=0.0895, interpMode=RCIM_Cubic, leaveTangent=0.0895, tangentMode=RCTM_User, time=0.1, value=1}
         {arriveTangent=-0.8586, interpMode=RCIM_Cubic, leaveTangent=-0.8586, tangentMode=RCTM_User, time=0.815, value=0.72}
         {arriveTangent=-4.4162, interpMode=RCIM_Cubic, leaveTangent=-4.4162, tangentMode=RCTM_User, time=1, value=0}
        ]}}}
     Scale Curve=&Multiply_Float(
      A=&FN_FrontEndFloatModifier(Gameplay=5, Front End=5)
      B=&FNTimeOfDayScalar(Night Scale=0.5, Morning Scale=1, Day Scale=0.75, Evening Scale=0.8)
     )
    )
   )
   &AlignSpriteToMeshOrientation(Mesh Orientation Relative Sprite Facing Vector:Vector3f=<0 0 -1>, SetAlignment=true, SetFacing=true)
  ]
  rendererClasses [&NiagaraSpriteRendererProperties, &NiagaraMeshRendererProperties]
  renderers [
   {rendererClass=&NiagaraSpriteRendererProperties}
   {rendererClass=&NiagaraMeshRendererProperties, rendererIndex=1}
  ]
 }
 Crosshair001 {
  emitterUpdateScript [
   &EmitterState()
   &SpawnBurst_Instantaneous(Spawn Count=5, Spawn Group=1, Spawn Time=0.5, Use Spawn Probability=false, UseLoopCountLimit=false) {enabled=false, module=SpawnBurst_Instantaneous002}
   &SpawnBurst_Instantaneous(Spawn Count=5, Spawn Group=1, Spawn Time=1.72, Use Spawn Probability=false, UseLoopCountLimit=false) {module=SpawnBurst_Instantaneous001}
   &SpawnBurst_Instantaneous(Spawn Time=0.01, Use Spawn Probability=false, UseLoopCountLimit=false)
   /SetVariables(
    Emitter.Layer 1 - Main reticle:LinearColor=<rgba 1 0 0 1>
    Emitter.Layer 2 - Reticle outline glow:LinearColor=<rgba 1 0.3127 0 1>
    Emitter.Layer 4 - outerglow:LinearColor=<rgba 0.3118 0.4223 1 1>
   ) {isSetParameters=true, module=SetVariables_#7}
  ]
  particleSpawnScript [
   /SetVariables(Particles.VisibilityTag=&SpawnGroup()) {isSetParameters=true, module=SetVariables_#8}
   &InitializeParticle(
    Lifetime=&MergeHelper_FloatSwitch(0=2.75, 1=0.3, 2=?, 3=?, Index=&SpawnGroup(), Num=::NewEnumerator0)
    Mass Max=2
    Mass Mode=::NewEnumerator3
    Mesh Scale=<2 1.4 2>
    Mesh Scale Mode=::NewEnumerator4
    Sprite Rotation Mode=::NewEnumerator1
    Sprite Size Mode=::NewEnumerator3
    Uniform Sprite Size=45
    UsePositionOffset=true
    Write Material Random=true
   )
   &ApplyInitialForces()
   &SkeletalMeshLocation(
    Bone Sampling Mode=::NewEnumerator4
    MaskSpawnGroup=false
    Sampled Orientation Coordinate Space=::Local
    Sampled Position Coordinate Space=::Local
    Sampled Position Offset=&MergeHelper_VectorSwitch(0=<0 0 0>, 1=<0 0 0>, 2=?, 3=?, Index=&SpawnGroup(), Num=::NewEnumerator0)
    Sampled Position Offset Coordinate Space:ENiagaraSkelMeshTransforms=::NewEnumerator2
    Sampled Velocity Coordinate Space:ENiagaraCoordinateSpace=::Local
    Skeletal Mesh=$System.Skeletal Mesh Reader
   ) {enabled=false}
   /SetVariables() {isSetParameters=true, module=SetVariables_#9}
  ]
  particleUpdateScript [
   &SolveForcesAndVelocity(Clamp Velocity=false, Limit Acceleration=false)
   &ParticleState()
   &ScaleColor(
    Scale Alpha=&FloatFromCurve(FloatCurve=DataInterface() {propertyValues {Curve {keys [
         {arriveTangent=0.002, interpMode=RCIM_Cubic, time=0, value=1.0113}
         {arriveTangent=-2.705, interpMode=RCIM_Cubic, leaveTangent=-2.705, tangentMode=RCTM_User, time=1.0011, value=0}
        ]}}})
    ScaleA=true
    ScaleRGB=false
   )
   /SetVariables(Particles.SpriteSize=&Multiply_Vector2D(
     A=$Particles.Initial.SpriteSize
     B=&MakeVector2D(
      X=&FloatFromCurve(FloatCurve=DataInterface() {propertyValues {Curve {keys [
           <at 0.015 0 RCIM_Cubic>
           {arriveTangent=33.3333, interpMode=RCIM_Cubic, leaveTangent=33.3333, time=0.03, value=1.3}
           {arriveTangent=-0.5263, interpMode=RCIM_Cubic, leaveTangent=-0.5263, time=0.045, value=1}
           {arriveTangent=0.5217, interpMode=RCIM_Cubic, leaveTangent=0.5217, time=0.6, value=1}
           {arriveTangent=-0.3208, interpMode=RCIM_Cubic, leaveTangent=-0.3208, tangentMode=RCTM_User, time=0.62, value=1.3}
           <at 0.65 0 RCIM_Cubic>
          ]}}})
      Y=&FloatFromCurve(FloatCurve=DataInterface() {propertyValues {Curve {keys [
           <at 0 2 RCIM_Cubic>
           {arriveTangent=-22.2222, interpMode=RCIM_Cubic, leaveTangent=-22.2222, time=0.025, value=0.8}
           {arriveTangent=0.3478, interpMode=RCIM_Cubic, leaveTangent=0.3478, time=0.045, value=1}
           {arriveTangent=-0.3478, interpMode=RCIM_Cubic, leaveTangent=-0.3478, time=0.6, value=1}
           {arriveTangent=0.2037, interpMode=RCIM_Cubic, leaveTangent=0.2037, tangentMode=RCTM_User, time=0.62, value=0.8}
           <at 0.65 2 RCIM_Cubic>
          ]}}})
     )
    )) {isSetParameters=true, module=SetVariables_#10}
   /SetVariables(
    Particles.Croshair main color:LinearColor=&MakeLinearColorFromVectorAndFloat(
     "Vector (RGB)"=&MakeVectorFromLinearColorRGB(Color=&FNTimeOfDayColor(Night Color=<rgba 1 0 0.0001 1>, Morning Color=<rgba 1 0 0.0001 1>, Day Color=<rgba 1 0 0.0001 1>, Evening Color=<rgba 1 0 0.0001 1>))
     "Float (Alpha)"=&FloatFromCurve(FloatCurve=DataInterface() {propertyValues {Curve {keys [{arriveTangent=0.002, interpMode=RCIM_Cubic, time=0, value=1.0113}, <at 1.0011 0 RCIM_Cubic>]}}})
    )
    Particles.Crosshair SpriteFacing:Vector3f=&Multiply_Vector(A=Linked() {linkedVariable {name=Particles.SkeletalMeshLocation.SampledPosition, type=NiagaraPosition}}, B=<1 0 0>)
    Particles.Crosshair glow SpriteSize:Vector2f=&Multiply_Vector2D(
     A=$Particles.Initial.SpriteSize
     B=&MakeVector2D(
      X=&FloatFromCurve(
       FloatCurve=DataInterface() {propertyValues {Curve {keys [
           <at 0.015 0 RCIM_Cubic>
           {arriveTangent=33.3333, interpMode=RCIM_Cubic, leaveTangent=33.3333, time=0.03, value=1.3}
           {arriveTangent=-0.5172, interpMode=RCIM_Cubic, leaveTangent=-0.5172, time=0.045, value=1}
           {arriveTangent=-0.3208, interpMode=RCIM_Cubic, leaveTangent=-0.3208, tangentMode=RCTM_User, time=0.61, value=1}
           <at 0.64 0 RCIM_Cubic>
          ]}}}
       Scale Curve=2.2
      )
      Y=&FloatFromCurve(
       FloatCurve=DataInterface() {propertyValues {Curve {keys [
           <at 0 2 RCIM_Cubic>
           {arriveTangent=-22.2222, interpMode=RCIM_Cubic, leaveTangent=-22.2222, time=0.025, value=0.8}
           {arriveTangent=0.3361, interpMode=RCIM_Cubic, leaveTangent=0.3361, time=0.045, value=1}
           {arriveTangent=0.2037, interpMode=RCIM_Cubic, leaveTangent=0.2037, tangentMode=RCTM_User, time=0.62, value=1}
           <at 0.635 2 RCIM_Cubic>
          ]}}}
       Scale Curve=2.2
      )
     )
    )
    Particles.Crosshair glow color:LinearColor=&MakeLinearColorFromVectorAndFloat(
     "Vector (RGB)"=&MakeVectorFromLinearColorRGB(Color=&FNTimeOfDayColor(Night Color=<rgba 1 0 0.0001 1>, Morning Color=<rgba 1 0 0.0001 1>, Day Color=<rgba 1 0 0.0001 1>, Evening Color=<rgba 1 0 0.0001 1>))
     "Float (Alpha)"=&FloatFromCurve(
      FloatCurve=DataInterface() {propertyValues {Curve {keys [{arriveTangent=0.002, interpMode=RCIM_Cubic, time=0, value=1.0113}, <at 0.9 0 RCIM_Cubic>]}}}
      Scale Curve=0.5
     )
    )
    Particles.Crosshair opposite SpriteFacing:Vector3f=<0 0 0>
    Particles.CrosshairGlow:LinearColor=&MakeLinearColorFromVectorAndFloat(
     "Vector (RGB)"=&MakeVectorFromLinearColorRGB(Color=&FNTimeOfDayColor(
       Night Color=<rgba 5 0.257 0.0079 1>
       Morning Color=<rgba 5 0.257 0.0079 1>
       Day Color=<rgba 5 0.257 0.0079 1>
       Evening Color=<rgba 5 0.257 0.0079 1>
      ))
     "Float (Alpha)"=&FloatFromCurve(
      FloatCurve=DataInterface() {propertyValues {Curve {keys [
          {arriveTangent=0.002, interpMode=RCIM_Cubic, time=0, value=1}
          <at 0.2549 -0.0093 RCIM_Cubic>
          {arriveTangent=0.0127, interpMode=RCIM_Cubic, leaveTangent=0.0127, time=0.5, value=1}
          <at 0.9918 0 RCIM_Cubic>
         ]}}}
      Scale Curve=0.5
     )
    )
    Particles.White details:LinearColor=&MakeLinearColorFromVectorAndFloat(
     "Vector (RGB)"=&MakeVectorFromLinearColorRGB(Color=&FNTimeOfDayColor(Night Color=<rgba 5 5 5 1>, Morning Color=<rgba 5 5 5 1>, Day Color=<rgba 5 5 5 1>, Evening Color=<rgba 5 5 5 1>))
     "Float (Alpha)"=&FloatFromCurve(
      FloatCurve=DataInterface() {propertyValues {Curve {keys [
          {arriveTangent=0.002, interpMode=RCIM_Cubic, time=0, value=1}
          <at 0.2549 -0.0093 RCIM_Cubic>
          {arriveTangent=0.0127, interpMode=RCIM_Cubic, leaveTangent=0.0127, time=0.5, value=1}
          <at 0.9918 0 RCIM_Cubic>
         ]}}}
      Scale Curve=0.5
     )
    )
    Particles.crosshair glow aligned SpriteSize:Vector2f=&Multiply_Vector2D(
     A=$Particles.Initial.SpriteSize
     B=&MakeVector2D(
      X=&FloatFromCurve(
       FloatCurve=DataInterface() {propertyValues {Curve {keys [
           <at 0.015 0 RCIM_Cubic>
           {arriveTangent=33.3333, interpMode=RCIM_Cubic, leaveTangent=33.3333, time=0.03, value=1.3}
           {arriveTangent=-0.5172, interpMode=RCIM_Cubic, leaveTangent=-0.5172, time=0.045, value=1}
           {arriveTangent=-0.3208, interpMode=RCIM_Cubic, leaveTangent=-0.3208, tangentMode=RCTM_User, time=0.61, value=1}
           <at 0.64 0 RCIM_Cubic>
          ]}}}
       Scale Curve=1.5
      )
      Y=&FloatFromCurve(
       FloatCurve=DataInterface() {propertyValues {Curve {keys [
           <at 0 2 RCIM_Cubic>
           {arriveTangent=-22.2222, interpMode=RCIM_Cubic, leaveTangent=-22.2222, time=0.025, value=0.8}
           {arriveTangent=0.3361, interpMode=RCIM_Cubic, leaveTangent=0.3361, time=0.045, value=1}
           {arriveTangent=0.2037, interpMode=RCIM_Cubic, leaveTangent=0.2037, tangentMode=RCTM_User, time=0.62, value=1}
           <at 0.635 2 RCIM_Cubic>
          ]}}}
       Scale Curve=1.5
      )
     )
    )
   ) {isSetParameters=true, module=SetVariables_#11}
   /SetVariables(Particles.Crosshair Main DynParam:Vector4f=&MakeVector4(X=0, Y=2, Z=0, W=0)) {isSetParameters=true, module=SetVariables_#12}
   /SetVariables(
    Particles.Flare Position:NiagaraPosition=&AddVectorToPosition(Position=$Particles.SkeletalMeshLocation.SampledPosition, Vector=<0 0 0>)
    Particles.Flare SpriteSize:Vector2f=&Multiply_Vector2D(
     A=$Particles.Initial.SpriteSize
     B=&MakeVector2D(
      X=&FloatFromCurve(
       FloatCurve=DataInterface() {propertyValues {Curve {defaultValue=1, keys [{tangentMode=RCTM_User, time=0, value=1}, {tangentMode=RCTM_User, time=1, value=0}]}}}
       Scale Curve=&FN_FrontEndFloatModifier(Gameplay=5, Front End=1)
      )
      Y=&FloatFromCurve(
       FloatCurve=DataInterface() {propertyValues {Curve {keys [<at 0.0122 0.8471>,<at 1 0>]}}}
       Scale Curve=&FN_FrontEndFloatModifier(Gameplay=5, Front End=1)
      )
     )
    )
    Particles.Sparkle Linear Color:LinearColor=&MakeLinearColorFromVectorAndFloat(
     "Vector (RGB)"=&MakeVectorFromLinearColorRGB(Color=&FNTimeOfDayColor(Night Color=<rgba 1 0 0.0001 1>, Morning Color=<rgba 1 0 0.0001 1>, Day Color=<rgba 1 0 0.0001 1>, Evening Color=<rgba 1 0 0.0001 1>))
     "Float (Alpha)"=&FloatFromCurve(FloatCurve=DataInterface() {propertyValues {Curve {keys [{arriveTangent=0.002, interpMode=RCIM_Cubic, time=0, value=1.0113}, <at 0.9 0 RCIM_Cubic>]}}})
    )
    Particles.Sparkle SpriteRotation:NiagaraFloat=&Multiply_Float(A=$Particles.NormalizedAge, B=60)
    Particles.Sparkle SpriteSize:Vector2f=&Multiply_Vector2D(
     A=$Particles.Initial.SpriteSize
     B=&Multiply_Vector2D_ByFloat(
      Float=&FloatFromCurve(FloatCurve=DataInterface() {propertyValues {Curve {keys [<at 0 1.5>,<at 1 0>]}}})
      Vector2D=<1 1>
     )
    )
   ) {isSetParameters=true, module=SetVariables_#13}
   /SetVariables(
    Particles.DynParam 0:Vector4f=&MakeVector4(X=1, Y=0, Z=0.1, W=0.5)
    Particles.Dynparam 1:Vector4f=&MakeVector4(X=1, Y=1, Z=1, W=1)
   ) {isSetParameters=true, module=SetVariables_#14}
   &FN_Light_Attributes(
    Color=<rgba 1 0 0 1>
    Create Color Attribute=true
    Create Exponent Attribute=true
    Create Position Attribute=true
    Create Radius Attribute=true
    Enable Particle Light Enable=true
    Enable Visibility Tag=true
    Exponent=6
    Light Enabled=true
    Position:NiagaraPosition=$Particles.SkeletalMeshLocation.SampledPosition
    Radius=&Multiply_Float(
     A=30
     B=&FloatFromCurve(FloatCurve=DataInterface() {propertyValues {Curve {keys [
          <at 0.015 0 RCIM_Cubic>
          {arriveTangent=33.3333, interpMode=RCIM_Cubic, leaveTangent=33.3333, time=0.03, value=1.3}
          {arriveTangent=-0.5263, interpMode=RCIM_Cubic, leaveTangent=-0.5263, time=0.045, value=1}
          {arriveTangent=0.5217, interpMode=RCIM_Cubic, leaveTangent=0.5217, time=0.6, value=1}
          {arriveTangent=-0.3208, interpMode=RCIM_Cubic, leaveTangent=-0.3208, tangentMode=RCTM_User, time=0.62, value=1.3}
          <at 0.65 0 RCIM_Cubic>
         ]}}})
    )
    Scale Color=&FloatFromCurve(
     FloatCurve=DataInterface() {propertyValues {Curve {keys [
         {arriveTangent=5.626, interpMode=RCIM_Cubic, leaveTangent=5.626, tangentMode=RCTM_User, time=0, value=0.6}
         {arriveTangent=0.0895, interpMode=RCIM_Cubic, leaveTangent=0.0895, tangentMode=RCTM_User, time=0.1, value=1}
         {arriveTangent=-0.8586, interpMode=RCIM_Cubic, leaveTangent=-0.8586, tangentMode=RCTM_User, time=0.815, value=0.72}
         {arriveTangent=-4.4162, interpMode=RCIM_Cubic, leaveTangent=-4.4162, tangentMode=RCTM_User, time=1, value=0}
        ]}}}
     Scale Curve=&Multiply_Float(
      A=&FN_FrontEndFloatModifier(Gameplay=5, Front End=1)
      B=&FNTimeOfDayScalar(Night Scale=0.5, Morning Scale=1, Day Scale=0.75, Evening Scale=0.8)
     )
    )
   )
   &AlignSpriteToMeshOrientation(Mesh Orientation Relative Sprite Facing Vector:Vector3f=<0 0 1>, SetAlignment=true, SetFacing=true)
  ]
  rendererClasses [
   &NiagaraSpriteRendererProperties
   &NiagaraLightRendererProperties
   &NiagaraMeshRendererProperties
  ]
  renderers [
   {rendererClass=&NiagaraSpriteRendererProperties}
   {rendererClass=&NiagaraSpriteRendererProperties, rendererIndex=1}
   {rendererClass=&NiagaraSpriteRendererProperties, rendererIndex=2}
   {rendererClass=&NiagaraSpriteRendererProperties, rendererIndex=3}
   {rendererClass=&NiagaraSpriteRendererProperties, rendererIndex=4}
   {rendererClass=&NiagaraLightRendererProperties, rendererIndex=5}
   {rendererClass=&NiagaraMeshRendererProperties, rendererIndex=6}
  ]
 }
}
summary {
 systemName=NS_Emote_GeneralHound_Crosshair
 emitters [
  {
   emitterName=Crosshair001
   rendererClasses [
    &NiagaraSpriteRendererProperties
    &NiagaraLightRendererProperties
    &NiagaraMeshRendererProperties
   ]
  }
  {emitterName=Crosshair, rendererClasses [&NiagaraSpriteRendererProperties, &NiagaraMeshRendererProperties]}
 ]
 userVariables [
  User.bIsFrontEnd {type=NiagaraBool}
  User.bIsFrontEndPreview {type=NiagaraBool}
  User.NormalizedNotifyProgress {type=NiagaraFloat, defaultValue=0, description=""}
  User.TargetB {type=NiagaraFloat, defaultValue=1, description=""}
 ]
}
