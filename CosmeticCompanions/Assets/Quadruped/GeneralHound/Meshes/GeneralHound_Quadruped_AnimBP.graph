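# Animation Blueprint for the GeneralHound quadruped cosmetic companion, producing its pose through IK retargeting from another mesh.
# bUseReverseRetarget blends between the standard retargeter and an emote retargeter with additive root yaw.
# That yaw scales a random angle, cached outside the frontend while RandomRotate curves are active, by right minus left curve values.
# Each update also spawns or deactivates an attached crosshair Niagara effect from the NS_Active curve.
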AnimBlueprint /CosmeticCompanions/Assets/Quadruped/GeneralHound/Meshes/GeneralHound_Quadruped_AnimBP
target_skeleton=&/GeneralHound_Quadruped_Skeleton
graphs [
 AnimGraph {nodes [
   Copy Pose From Mesh() {bCopyCurves=true, bUseAttachedParent=true}
   n1 = GetCurveValue(CurveName=RandomRotateLeft)
   n2 = GetCurveValue(CurveName=RandomRotateRight)
   n3 = "Transform (Modify) Bone - Bone: root"!(
    ComponentPose=Local To Component(LocalPose=Retarget Pose From Mesh!() -> (Pose:Pose Link Structure) {
      is=title AnimGraphNode_RetargetPoseFromMesh
      iKRetargeterAsset=&/RTG_GeneralHound_Emote
     })
    Alpha
    Translation="0,0,0"
    Rotation_Roll:float=0
    Rotation_Pitch:float=0
    Rotation_Yaw:float=float * float(A=float - float(A=@n2, B=@n1), B=Get CachedRandomAngle() -> (CachedRandomAngle:double)) {is=title K2Node_PromotableOperator}
    Scale
   ) -> (Pose) {rotationMode=BMM_Additive, rotationSpace=BCS_BoneSpace, boneToModify {boneName=root}, rotation {yaw=40}}
   Output Pose(Result=Blend Poses by bool(
     BlendPose_0=Component To Local(ComponentPose=@n3)
     BlendPose_1=Retarget Pose From Mesh() {iKRetargeterAsset=&/RTG_GeneralHound}
     BlendTime_0=0.25
     BlendTime_1=0.25
     bActiveValue=Get bUseReverseRetarget() -> (bUseReverseRetarget:bool)
    ))
   n4 = Get bRandomAngleSet() -> (bRandomAngleSet:bool)
   SelectFloat(
    A=RandomFloatInRange(Min=0, Max=45)
    B=Get CachedRandomAngle()
    bPickA=AND Boolean(A="float > float"(A:Wildcard, B:Wildcard), B=NOT Boolean(A=@n4))
   )
  ]}
 EventGraph {nodes [
   Event BlueprintUpdateAnimation()
   > RandomRotation()
   > Niagara()
   TryGetPawnOwner()
  ]}
 Niagara {nodes [
   Niagara() {is=title K2Node_FunctionEntry}
   > n1 = Branch(Condition="float > float"(A:double=GetCurveValue(CurveName=NS_Active), B:double))
   > n2 = Branch(Condition=Get bNS_Spawned() -> (bNS_Spawned:bool))
   n3 = SpawnSystemAttached!(
    execute=@n2.else
    SystemTemplate:Niagara System Object Reference=/CosmeticCompanions/FX/Builtins/GeneralHound/Emote/NS_Emote_GeneralHound_Crosshair
    AttachToComponent:Scene Component Object Reference=GetOwningComponent()
    AttachPointName:Name=attach_emote_01
    Location:Vector="0, 0, 0"
    Rotation:Rotator="0, 0, 0"
    LocationType:EAttachLocation Enum=SnapToTarget
    bAutoDestroy:bool=true
    bAutoActivate:bool=true
    PoolingMethod:ENCPoolMethod Enum=None
    bPreCullCheck:bool=true
   ) -> (then, ReturnValue:Niagara Particle System Component Object Reference)
   > Set SpawnedNS(SpawnedNS:Niagara Particle System Component Object Reference=@n3) -> (Output_Get:Niagara Particle System Component Object Reference)
   > Set bNS_Spawned(bNS_Spawned:bool=true) -> (Output_Get:bool)
   Branch(execute=@n1.else, Condition=Get bNS_Spawned())
   > Deactivate(self:Actor Component Object Reference=Get SpawnedNS() -> (SpawnedNS:Niagara Particle System Component Object Reference))
   > Set bNS_Spawned(bNS_Spawned=false)
  ]}
 RandomRotation {nodes [
   RandomRotation() {is=title K2Node_FunctionEntry}
   n1 = IsInFrontend!() -> (ReturnValue:bool)
   n2 = float + float(A=GetCurveValue(CurveName=RandomRotateRight), B=GetCurveValue(CurveName=RandomRotateLeft))
   > n3 = Branch(Condition=AND Boolean(A=NOT Boolean(A=@n1), B="float > float"(A=@n2)))
   > n4 = Branch(Condition=Get bRandomAngleSet())
   Set CachedRandomAngle(execute=@n4.else, CachedRandomAngle:double=RandomFloatInRange(Min=0, Max=90)) -> (Output_Get:double)
   > Set bRandomAngleSet(bRandomAngleSet:bool=true) -> (Output_Get:bool)
   Set bRandomAngleSet(execute=@n3.else, bRandomAngleSet=false)
  ]}
]
variables [
 bNS_Spawned {type=boolean}
 bRandomAngleSet {type=boolean}
 CachedRandomAngle {type=number}
 SpawnedNS {type=NiagaraComponent, default=null}
]
