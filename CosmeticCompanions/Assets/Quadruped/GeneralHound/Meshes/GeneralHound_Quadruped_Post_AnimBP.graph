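# Post-process AnimBP for the GeneralHound quadruped companion skeleton, adding corrective posing and secondary motion to the bandana, vest handle and upper lips.
# The input pose runs through a PoseAsset blend node, then weighted constraints drive the ribbon and handle base bones from neck and spine, plus front scapulae for the handle.
# A chain of root-relative, spring-driven AnimDynamics nodes then simulates their dyn_ bones and six lip bones.
# Cached poses separate the stages; the EventGraph is empty.
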AnimBlueprint /CosmeticCompanions/Assets/Quadruped/GeneralHound/Meshes/GeneralHound_Quadruped_Post_AnimBP
target_skeleton=&/GeneralHound_Quadruped_Skeleton
graphs [
 AnimGraph {nodes [
   Output Pose(Result=Use cached pose 'DynamicsCache'())
   Save cached pose 'InputCache'(Pose=Pose!(
     PoseAsset:Pose Asset Object Reference=Get PoseAsset() -> (PoseAsset:Pose Asset Object Reference)
     SourcePose:Pose Link Structure=Input Pose()
    ) -> (Pose:Pose Link Structure) {is=title AnimGraphNode_PoseBlendNode, bIgnorePoseAssetSourceAnimationCurves=true, blendOption=Linear}) {cacheName=InputCache}
   n1 = "Constraint - base_bandanaRibbon_r (PPPP)"!(
    ComponentPose=Local To Component(LocalPose=Use cached pose 'InputCache'())
    Alpha=1
    ConstraintWeights_0=0.437
    ConstraintWeights_1:float=0.121
    ConstraintWeights_2:float=0.068
    ConstraintWeights_3:float=0.374
   ) -> (Pose) {
    boneToModify {boneName=base_bandanaRibbon_r}
    constraintSetup [{targetBone {boneName=neck_01}}, {targetBone {boneName=spine_03}}, {targetBone {boneName=spine_02}}, {targetBone {boneName=neck_02}}]
    constraintWeights [1,1,1,1]
   }
   n2 = "Constraint - base_bandanaRibbon_l (PPPP)"!(
    ComponentPose=@n1
    Alpha=1
    ConstraintWeights_0=0.437
    ConstraintWeights_1:float=0.121
    ConstraintWeights_2:float=0.068
    ConstraintWeights_3:float=0.374
   ) -> (Pose) {
    boneToModify {boneName=base_bandanaRibbon_l}
    constraintSetup [{targetBone {boneName=neck_01}}, {targetBone {boneName=spine_03}}, {targetBone {boneName=spine_02}}, {targetBone {boneName=neck_02}}]
    constraintWeights [1,1,1,1]
   }
   n3 = "Constraint - base_vestHandle (PPPPPT)"!(
    ComponentPose=@n2
    Alpha=1
    ConstraintWeights_0=0.462
    ConstraintWeights_1:float=0.24
    ConstraintWeights_2:float=0.19
    ConstraintWeights_3:float=0.08
    ConstraintWeights_4:float=0.014
    ConstraintWeights_5:float=0.014
   ) -> (Pose) {
    boneToModify {boneName=base_vestHandle}
    constraintSetup [
     {targetBone {boneName=neck_01}}
     {targetBone {boneName=spine_03}}
     {targetBone {boneName=spine_02}}
     {targetBone {boneName=neck_02}}
     {targetBone {boneName=scapula_fr_l}}
     {transformType=Translation, targetBone {boneName=scapula_fr_r}}
    ]
    constraintWeights [1,1,1,1,1,1]
   }
   Save cached pose 'PreDynamics_Cache'(Pose=Component To Local(ComponentPose=@n3)) {cacheName=PreDynamics_Cache}
   n4 = Anim Dynamics - Bone: dyn_bandanaRibbon_l!(ComponentPose=Local To Component(LocalPose=Use cached pose 'PreDynamics_Cache'()), Alpha=1, GravityOverride:Vector="0,0,-100") -> (Pose) {
    angularDampingOverride=0.9999
    angularSpringConstant=8000
    bAngularSpring=true
    bOverrideAngularDamping=true
    bUseGravityOverride=true
    gravityScale=0.35
    simulationSpace=RootRelative
    alphaBoolBlend {blendOption=Linear}
    boundBone {boneName=dyn_bandanaRibbon_l}
    componentAppliedLinearAccClamp=<5000 5000 5000>
    componentLinearAccScale=<1 1 1>
    componentLinearVelScale=<1 1 1>
    gravityOverride {z=-980}
    physicsBodyDefinitions [{
      boundBone {boneName=dyn_bandanaRibbon_l}
      boxExtents=<25 35 35>
      constraintSetup {angularLimitsMax=<10 35 60>, angularLimitsMin=<-10 -25 -25>, angularTarget=<1 0 0>}
      localJointOffset=<3 0 0>
     }]
    simSpaceSettings {maxAngularAcceleration=30, maxAngularVelocity=6, simSpaceAngularAlpha=1}
   }
   n5 = Anim Dynamics - Bone: dyn_bandanaRibbon_r!(ComponentPose=@n4, Alpha=1, GravityOverride:Vector="0,0,-100") -> (Pose) {
    angularDampingOverride=0.9999
    angularSpringConstant=8000
    bAngularSpring=true
    bOverrideAngularDamping=true
    bUseGravityOverride=true
    gravityScale=0.35
    simulationSpace=RootRelative
    alphaBoolBlend {blendOption=Linear}
    boundBone {boneName=dyn_bandanaRibbon_r}
    componentAppliedLinearAccClamp=<5000 5000 5000>
    componentLinearAccScale=<1 1 1>
    componentLinearVelScale=<1 1 1>
    gravityOverride {z=-980}
    physicsBodyDefinitions [{
      boundBone {boneName=dyn_bandanaRibbon_r}
      boxExtents=<25 35 35>
      constraintSetup {angularLimitsMax=<10 25 60>, angularLimitsMin=<-10 -35 -45>, angularTarget=<1 0 0>}
      localJointOffset=<3 0 0>
     }]
    simSpaceSettings {maxAngularAcceleration=30, maxAngularVelocity=6, simSpaceAngularAlpha=1}
   }
   n6 = Anim Dynamics - Bone: dyn_vestHandle!(ComponentPose=@n5, Alpha=1, GravityOverride:Vector="0,0,-980") -> (Pose) {
    angularDampingOverride=1
    angularSpringConstant=10000
    bAngularSpring=true
    bOverrideAngularDamping=true
    bUseGravityOverride=true
    gravityScale=0.35
    simulationSpace=RootRelative
    alphaBoolBlend {blendOption=Linear}
    boundBone {boneName=dyn_vestHandle}
    componentAppliedLinearAccClamp=<10000 10000 10000>
    componentLinearAccScale=<1 1 1>
    componentLinearVelScale=<1 1 1>
    gravityOverride {z=-980}
    physicsBodyDefinitions [{
      boundBone {boneName=dyn_vestHandle}
      boxExtents=<27 37 37>
      constraintSetup {angularLimitsMax=<2 25 2>, angularLimitsMin=<-2 -25 -30>, angularTarget=<1 0 0>}
      localJointOffset=<2 0 0>
     }]
    simSpaceSettings {maxAngularAcceleration=30, maxAngularVelocity=6, simSpaceAngularAlpha=1}
   }
   n7 = Anim Dynamics - Bone: dyn_lip_upper_01_l!(ComponentPose=@n6, Alpha=1, GravityOverride:Vector="0,0,-75") -> (Pose) {
    angularDampingOverride=1
    angularSpringConstant=5000
    bAngularSpring=true
    bOverrideAngularDamping=true
    bUseGravityOverride=true
    gravityScale=0.35
    simulationSpace=RootRelative
    alphaBoolBlend {blendOption=Linear}
    boundBone {boneName=dyn_lip_upper_01_l}
    componentAppliedLinearAccClamp=<1000 1000 1000>
    componentLinearAccScale=<1 1 1>
    componentLinearVelScale=<1 1 1>
    gravityOverride {z=-980}
    physicsBodyDefinitions [{
      boundBone {boneName=dyn_lip_upper_01_l}
      boxExtents=<30 30 20>
      constraintSetup {angularTargetAxis=AxisZ, twistAxis=AxisZ, angularLimitsMax=<0 0 0>, angularLimitsMin=<0 -90 0>, angularTarget=<0 0 1>}
      localJointOffset=<0 0 2.5>
     }]
    simSpaceSettings {maxAngularAcceleration=33, maxAngularVelocity=6, simSpaceAngularAlpha=1}
   }
   n8 = Anim Dynamics - Bone: dyn_lip_upper_02_l!(ComponentPose=@n7, Alpha=1, GravityOverride:Vector="0,0,-75") -> (Pose) {
    angularDampingOverride=1
    angularSpringConstant=5000
    bAngularSpring=true
    bOverrideAngularDamping=true
    bUseGravityOverride=true
    gravityScale=0.35
    simulationSpace=RootRelative
    alphaBoolBlend {blendOption=Linear}
    boundBone {boneName=dyn_lip_upper_02_l}
    componentAppliedLinearAccClamp=<1000 1000 1000>
    componentLinearAccScale=<1 1 1>
    componentLinearVelScale=<1 1 1>
    gravityOverride {z=-980}
    physicsBodyDefinitions [{
      boundBone {boneName=dyn_lip_upper_02_l}
      boxExtents=<30 30 20>
      constraintSetup {angularTargetAxis=AxisZ, twistAxis=AxisZ, angularLimitsMax=<0 0 0>, angularLimitsMin=<0 -90 0>, angularTarget=<0 0 1>}
      localJointOffset=<0 0 2.5>
     }]
    simSpaceSettings {maxAngularAcceleration=33, maxAngularVelocity=6, simSpaceAngularAlpha=1}
   }
   n9 = Anim Dynamics - Bone: dyn_lip_upper_03_l!(ComponentPose=@n8, Alpha=1, GravityOverride:Vector="0,0,-75") -> (Pose) {
    angularDampingOverride=1
    angularSpringConstant=5000
    bAngularSpring=true
    bOverrideAngularDamping=true
    bUseGravityOverride=true
    gravityScale=0.35
    simulationSpace=RootRelative
    alphaBoolBlend {blendOption=Linear}
    boundBone {boneName=dyn_lip_upper_03_l}
    componentAppliedLinearAccClamp=<1000 1000 1000>
    componentLinearAccScale=<1 1 1>
    componentLinearVelScale=<1 1 1>
    gravityOverride {z=-980}
    physicsBodyDefinitions [{
      boundBone {boneName=dyn_lip_upper_03_l}
      boxExtents=<30 30 20>
      constraintSetup {angularTargetAxis=AxisZ, twistAxis=AxisZ, angularLimitsMax=<0 0 0>, angularLimitsMin=<0 -90 0>, angularTarget=<0 0 1>}
      localJointOffset=<0 0 2.5>
     }]
    simSpaceSettings {maxAngularAcceleration=33, maxAngularVelocity=6, simSpaceAngularAlpha=1}
   }
   n10 = Anim Dynamics - Bone: dyn_lip_upper_01_r!(ComponentPose=@n9, Alpha=1, GravityOverride:Vector="0,0,-75") -> (Pose) {
    angularDampingOverride=1
    angularSpringConstant=5000
    bAngularSpring=true
    bOverrideAngularDamping=true
    bUseGravityOverride=true
    gravityScale=0.35
    simulationSpace=RootRelative
    alphaBoolBlend {blendOption=Linear}
    boundBone {boneName=dyn_lip_upper_01_r}
    componentAppliedLinearAccClamp=<1000 1000 1000>
    componentLinearAccScale=<1 1 1>
    componentLinearVelScale=<1 1 1>
    gravityOverride {z=-980}
    physicsBodyDefinitions [{
      boundBone {boneName=dyn_lip_upper_01_r}
      boxExtents=<30 30 20>
      constraintSetup {angularTargetAxis=AxisZ, twistAxis=AxisZ, angularLimitsMax=<0 0 0>, angularLimitsMin=<0 -90 0>, angularTarget=<0 0 1>}
      localJointOffset=<0 0 -2.5>
     }]
    simSpaceSettings {maxAngularAcceleration=33, maxAngularVelocity=6, simSpaceAngularAlpha=1}
   }
   n11 = Anim Dynamics - Bone: dyn_lip_upper_02_r!(ComponentPose=@n10, Alpha=1, GravityOverride:Vector="0,0,-75") -> (Pose) {
    angularDampingOverride=1
    angularSpringConstant=5000
    bAngularSpring=true
    bOverrideAngularDamping=true
    bUseGravityOverride=true
    gravityScale=0.35
    simulationSpace=RootRelative
    alphaBoolBlend {blendOption=Linear}
    boundBone {boneName=dyn_lip_upper_02_r}
    componentAppliedLinearAccClamp=<1000 1000 1000>
    componentLinearAccScale=<1 1 1>
    componentLinearVelScale=<1 1 1>
    gravityOverride {z=-980}
    physicsBodyDefinitions [{
      boundBone {boneName=dyn_lip_upper_02_r}
      boxExtents=<30 30 20>
      constraintSetup {angularTargetAxis=AxisZ, twistAxis=AxisZ, angularLimitsMax=<0 0 0>, angularLimitsMin=<0 -90 0>, angularTarget=<0 0 1>}
      localJointOffset=<0 0 -2.5>
     }]
    simSpaceSettings {maxAngularAcceleration=33, maxAngularVelocity=6, simSpaceAngularAlpha=1}
   }
   n12 = Anim Dynamics - Bone: dyn_lip_upper_03_r!(ComponentPose=@n11, Alpha=1, GravityOverride:Vector="0,0,-75") -> (Pose) {
    angularDampingOverride=1
    angularSpringConstant=5000
    bAngularSpring=true
    bOverrideAngularDamping=true
    bUseGravityOverride=true
    gravityScale=0.35
    simulationSpace=RootRelative
    alphaBoolBlend {blendOption=Linear}
    boundBone {boneName=dyn_lip_upper_03_r}
    componentAppliedLinearAccClamp=<1000 1000 1000>
    componentLinearAccScale=<1 1 1>
    componentLinearVelScale=<1 1 1>
    gravityOverride {z=-980}
    physicsBodyDefinitions [{
      boundBone {boneName=dyn_lip_upper_03_r}
      boxExtents=<30 30 20>
      constraintSetup {angularTargetAxis=AxisZ, twistAxis=AxisZ, angularLimitsMax=<0 0 0>, angularLimitsMin=<0 -90 0>, angularTarget=<0 0 1>}
      localJointOffset=<0 0 -2.5>
     }]
    simSpaceSettings {maxAngularAcceleration=33, maxAngularVelocity=6, simSpaceAngularAlpha=1}
   }
   Save cached pose 'DynamicsCache'(Pose=Component To Local(ComponentPose=@n12)) {cacheName=DynamicsCache}
  ]}
 EventGraph {links [], nodes []}
]
variables [PoseAsset {type=PoseAsset, default=&/GeneralHound_Quadruped_PoseAsset}]
